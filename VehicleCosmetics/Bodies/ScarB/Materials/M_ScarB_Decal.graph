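# Masked Default Lit surface material for the decal on the ScarB vehicle body cosmetic.
# It exposes a Decal_BC texture parameter, defaulting to an LOD base color texture, plus Metallic, Specular and Roughness scalars.
# Usage is fixed to skeletal meshes and mesh deformers with automatic usage setting off, and ray traced shadow casting is disabled.
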Material /VehicleCosmetics/Bodies/ScarB/Materials/M_ScarB_Decal
graph {
 properties {
  bAutomaticallySetUsageInEditor=false
  bCastRayTracedShadows=false
  bUsedWithMeshDeformer=true
  bUsedWithSkeletalMesh=true
  blendMode=BLEND_Masked
  previewMesh=/Engine/Transient.MaterialEditorMeshComponent_7
  usedShadingModels=Default Lit
 }
 nodes [
  ScalarParameter() {parameterName=Metallic}
  ScalarParameter() {defaultValue=0.5, parameterName=Specular}
  ScalarParameter() {defaultValue=0.25, parameterName=Roughness}
  TextureSampleParameter2D() {parameterName=Decal_BC, texture=/VehicleCosmetics/Bodies/ScarB/Textures/LOD/T_ScarB_Decal_BC_LOD}
 ]
}
