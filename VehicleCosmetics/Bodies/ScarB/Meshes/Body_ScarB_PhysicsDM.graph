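# Physics asset for the ScarB vehicle body mesh, a cosmetic body variant under VehicleCosmetics.
# It has a single box body on the root bone, sized to the car's overall hull and offset slightly back and upward.
# There are no constraints, so the body is one rigid collision shape with no articulated parts.
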PhysicsAsset /VehicleCosmetics/Bodies/ScarB/Meshes/Body_ScarB_PhysicsDM
bodies {root [{shapeType=Box, extentX=480, extentY=216, extentZ=116, shapeName=root_box, center=<-10 0 91>}]}
constraints []
